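; Building prop Blueprint that assembles a base mesh, down rod and light fixture at an adjustable FixtureLength.
; The construction script reads mesh Z bounds, positions the fixture, and adds a capsule replacing part collision unless mesh collision is enabled.
; The rod either stretches with matching material UV scale or repeats as rotated instanced meshes, like chain links.
; Tint comes from building texture data or an emissive tint, with optional root-material copying to the fixture.
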(event EventBeginPlay)

(event Collision|EventActorBeginOverlap
  (params (OtherActor :type Actor%20Object%20Reference)))

(event EventTick
  (params (DeltaSeconds :type Float%20%28single-precision%29)))

(fn AddCapsuleCollision
  (bind _returnvalue (AddComponent|NODE_AddCapsuleComponent-0 :RelativeTransform (Math|Transform|MakeTransform (Math|Vector|MakeVector 0.0 0.0 (- (/ (Variables|Default|GetFixtureLength) 2.0) (- (Variables|Default|GetFixtureLength) (Variables|Default|GetBaseZBounds)))))))
  (bind _output_get (Variables|Default|SetCapsuleCollision _returnvalue))
  (Components|Capsule|SetCapsuleHalfHeight _output_get (/ (Variables|Default|GetFixtureLength) 2.0) true)
  (for _array_element (Utilities|Array|MakeArray (Variables|Default|GetLightFixture) (Variables|Default|GetDownRod) (Variables|Debug|GetStaticMeshComponent))
    (Collision|SetCollisionEnabled _array_element "NoCollision")))

(fn EnableCollisionOnMeshParts
  (for _array_element (Utilities|Array|MakeArray (Variables|Default|GetDownRod) (Variables|Default|GetLightFixture))
    (Collision|SetCollisionProfileName _array_element "Default" true :!class /Script/Engine.PrimitiveComponent)
    (Development|PrintString (Utilities|String|ToString(Name) (Collision|GetCollisionProfileName (stale _array_element_2)))))
  (for _array_element_2 (Utilities|Array|MakeArray (Variables|Default|GetDownRod) (Variables|Default|GetLightFixture))
    _))

(fn GetMaterialFromRootMesh
  (Rendering|Material|SetMaterial (Variables|Default|GetLightFixture) 0 (Material|GetMID (Variables|Debug|GetStaticMeshComponent))))

(fn GetMeshBounds
  (bind (_min _max) (Components|StaticMesh|GetLocalBounds (Variables|Default|GetDownRod)))
  (Variables|Default|SetBaseZBounds (.z (Components|StaticMesh|GetLocalBounds (Variables|Debug|GetStaticMeshComponent))))
  (Variables|Default|SetFixtureZBounds (.z (Components|StaticMesh|GetLocalBounds (Variables|Default|GetLightFixture))))
  (Variables|Default|SetRodZBounds (.z _max)))

(fn GetTintColorFromTextureData
  (bind _outtexturedataarray (Building|GetBuildingTextureData))
  (for _array_element (Utilities|Array|MakeArray (Variables|Debug|GetStaticMeshComponent) (Variables|Default|GetDownRod) (Variables|Default|GetLightFixture))
    (Default|SetColoronSecondaryMesh (Math|Conversions|ToLinearColor(Color) (Variables|/Script/FortniteGame.BuildingTextureData|GetTintColor (Utilities|Array|Get(acopy) _outtexturedataarray))) _array_element)))

(fn SetDownRodMaterial
  (bind _downrod (Variables|Default|GetDownRod))
  (bind _returnvalue (Rendering|Material|GetMaterial _downrod 0))
  (bind _downrod_2 (Variables|Default|GetDownRod))
  (bind _returnvalue_2 (Rendering|Material|GetMaterial _downrod_2 0))
  (if (Variables|Default|GetUseGeoForDownrod)
    (if (Utilities|IsValid _returnvalue_2 :!class /Script/Engine.KismetSystemLibrary)
      (Rendering|Material|CreateDynamicMaterialInstance (Variables|Default|GetISM_DownRod) 0 _returnvalue_2 "None" :!class /Script/Engine.PrimitiveComponent))
    (elif (Utilities|IsValid _returnvalue :!class /Script/Engine.KismetSystemLibrary)
      (bind _returnvalue_3 (Rendering|Material|CreateDynamicMaterialInstance (Variables|Default|GetDownRod) 0 _returnvalue "None" :!class /Script/Engine.PrimitiveComponent))
      (Rendering|Material|SetScalarParameterValue _returnvalue_3 "UV Scale Vertical" (Variables|Default|GetRodScale) :!class /Script/Engine.MaterialInstanceDynamic))))

(fn SetFixturePosition
  (Transformation|SetRelativeLocation (Variables|Default|GetLightFixture) (Math|Vector|MakeVector 0.0 0.0 (+ (Variables|Default|GetFixtureLength) (Variables|Default|GetFixtureZBounds))) false false))

(fn SetNumberofRodMeshes
  (bind _downrod (Variables|Default|GetDownRod))
  (bind _staticmesh (Variables|/Script/Engine.StaticMeshComponent|GetStaticMesh _downrod))
  (bind _returnvalue (AddComponent|NODE_AddInstancedStaticMeshComponent-0))
  (bind _output_get (Variables|Default|SetISM_DownRod _returnvalue))
  (if (Utilities|IsValid _staticmesh :!class /Script/Engine.KismetSystemLibrary)
    (Components|StaticMesh|SetStaticMesh _output_get _staticmesh)
    (bind _output_get_2 (LocalVariables|SetNumberOfInstances (/ (Utilities|Operators|Subtract (Variables|Default|GetFixtureLength) (* (Variables|Default|GetBaseZBounds) -1.0) (* (Variables|Default|GetFixtureZBounds) -1.0)) (* (Variables|Default|GetRodZBounds) -1.0))))
    (for _index (range (+ (Math|Float|Floor _output_get_2) 1))
      (Components|InstancedStaticMesh|AddInstance _output_get (Math|Transform|MakeTransform (Math|Vector|MakeVector 0.0 0.0 (+ (* _index (Variables|Default|GetRodZBounds)) (Variables|Default|GetBaseZBounds))) (Math|Rotator|MakeRotator 0.0 0.0 (Math|Conversions|ToFloat(Integer) (* _index (Variables|Components|GetChainlinkRotationAmount))))) false)
      (Rendering|SetVisibility (Variables|Default|GetDownRod) false false))
    (else
      (Development|PrintString "DownRod mesh not found"))))

(fn SetRodPosition
  (bind _rodzbounds (Variables|Default|GetRodZBounds))
  (if (> (Math|Float|Absolute(Float) _rodzbounds) 0.001)
    (Variables|Default|SetRodScale (/ (Utilities|Operators|Subtract (Variables|Default|GetFixtureLength) (* (Variables|Default|GetBaseZBounds) -1.0) (* (Variables|Default|GetFixtureZBounds) -1.0)) _rodzbounds))
    (else
      (Variables|Default|SetRodScale 1.0)))
  (Transformation|SetRelativeScale3D (Variables|Default|GetDownRod) (Math|Vector|MakeVector 1.0 1.0 (Variables|Default|GetRodScale)))
  (Transformation|SetRelativeLocation (Variables|Default|GetDownRod) (Math|Vector|MakeVector 0.0 0.0 (Variables|Default|GetBaseZBounds)) false false))

(fn SetTintColorOnMeshParts
  (for _array_element (Utilities|Array|MakeArray (Variables|Default|GetDownRod) (Variables|Default|GetLightFixture))
    ; IF TD Color Bool is true , Set TD as tint for each elements
    (if (Variables|Default|GetUse%20TD%20for%20Tint)
      (bind _outtexturedataarray (Building|GetBuildingTextureData))
      (if (> (Utilities|Array|Length _outtexturedataarray) 0)
        (for _array_element_2 (Components|GetChildrenComponents (Variables|Default|GetLightFixture) false)
          (bind _asstatic_mesh_component (Utilities|Casting|CastToStaticMeshComponent _array_element_2))
          (Default|SetColoronSecondaryMesh (Math|Conversions|ToLinearColor(Color) (Variables|/Script/FortniteGame.BuildingTextureData|GetTintColor (Utilities|Array|Get(acopy) _outtexturedataarray))) _asstatic_mesh_component)))
      (else
        (Default|SetColoronSecondaryMesh (Variables|Default|GetEmissiveTint) _array_element)))))

(fn ConstructionScript
  (Default|GetMeshBounds)
  (if (!= (Variables|Default|GetUseMeshCollision) true)
    (Default|AddCapsuleCollision))
  (Default|SetFixturePosition)
  (if (Variables|Default|GetUseGeoForDownrod)
    (Default|SetNumberofRodMeshes)
    (else
      (Default|SetRodPosition)))
  (Default|ValidateBuildingTextureData)
  (if (Variables|Default|GetHasTextureData)
    (Default|GetTintColorfromTextureData)
    (elif (Variables|Default|GetApplyRootMaterialToFixtureMesh)
      (Default|GetMaterialfromRootMesh)))
  (Default|SetDownRodMaterial)
  (Default|SetTintColoronMeshParts))

(fn ValidateBuildingTextureData
  (bind _outtexturedataarray (Building|GetBuildingTextureData))
  (Variables|Default|SetHasTextureData (> (Utilities|Array|Length _outtexturedataarray) 0))
  (return))
